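; Looping gameplay cue for the charge-up of the Stamina Hawaii beam item, driving its charge effect and sound indicator.
; On loop start it caches the first particle component and target pawn, then starts a repeating timer that shows a gunshot-type sound indicator ranged by the looping sound's max distance.
; Tick polls the pawn's Hawaii component replicated beam data to set the effect's ChargeState to charging or firing, clearing the indicator timer once firing begins.
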(event EventBeginPlay
  (|CallParentFunction))

(event EventOnLoopingStart(EitherVFX-preferthisversion)
  (params (MyTarget :type Actor%20Object%20Reference) (Parameters :type Gameplay%20Cue%20Parameters%20Structure%20%28by%20ref%29) (ParticleComponents :type Array%20of%20FXSystem%20Component%20Object%20References) (AudioComponents :type Array%20of%20Audio%20Component%20Object%20References))
  (bind _result (ArrayMacros|SafeArrayGet ParticleComponents)
    (:%20%20
      (Utilities|IsValid _result
        (:Is%20Valid
          (Variables|Default|SetChargeFX _result)
          (bind _asplayer_pawn_athena (Utilities|Casting|CastToPlayerPawn_Athena MyTarget))
          (Variables|Default|SetTargetPawn _asplayer_pawn_athena)
          (bind _returnvalue (Utilities|Time|SetTimerbyEvent (delegate Custom|ShowShoundIndicator) 0.3 true))
          (Variables|Default|SetSoundIndicatorTimer _returnvalue))))))

(event Custom|ShowShoundIndicator
  (Fortnite|ShowSoundIndicatorLocalforTeam 0 0 (Transformation|GetActorLocation (Variables|Default|GetTargetPawn)) (Variables|/Script/Engine.SoundBase|GetMaxDistance (Variables|/Script/Engine.AudioComponent|GetSound (Utilities|Array|Get(acopy) (stale EventOnLoopingStart(EitherVFX-preferthisversion) AudioComponents)))) "Gunshot" (Utilities|Array|MakeArray (Fortnite|GetFortPlayerControllerfromActor (Variables|Default|GetTargetPawn))) (Team|GetActorTeam (Variables|Default|GetTargetPawn)) (Utilities|Array|MakeArray "Friendly" "Neutral" "Hostile")))

(event EventTick
  (params (DeltaSeconds :type Float%20%28single-precision%29))
  (bind _targetpawn (Variables|Default|GetTargetPawn))
  (bind _returnvalue (Actor|GetComponentbyClass _targetpawn "/StaminaGameplay/Gameplay/Items/Hawaii/Stamina_Hawaii_Component.Stamina_Hawaii_Component_C"))
  (bind _soundindicatortimer (Variables|Default|GetSoundIndicatorTimer))
  (Utilities|IsValid (Variables|Default|GetChargeFX)
    (:Is%20Valid
      (Utilities|IsValid (Variables|Default|GetTargetPawn)
        (:Is%20Valid
          (Utilities|IsValid _returnvalue
            (:Is%20Valid
              (Utilities|FlowControl|Sequence
                (:then_0
                  (if (not (Utilities|Struct|BreakStaminaHawaiiBeamData (Variables|/StaminaGameplay/Gameplay/Items/Hawaii/Stamina_Hawaii_Component.Stamina_Hawaii_Component_C|GetRepBeamData _returnvalue)))
                    ;[1 1 1] Charging
                    (Variables|Default|SetChargeParam 0)
                    (else
                      ;[1 1 1] Firing
                      (Variables|Default|SetChargeParam 1)
                      (if (Utilities|Time|IsValidTimerHandle _soundindicatortimer)
                        (Utilities|Time|ClearandInvalidateTimerbyHandle 0 _soundindicatortimer)))))
                (:then_1
                  (Effects|Components|ParticleSystem|SetIntParameter (Variables|Default|GetChargeFX) "ChargeState" (Variables|Default|GetChargeParam)))))
            (:Is%20Not%20Valid
              (Variables|Default|SetChargeParam 0))))))))

(fn GetApllicationBurstSoundCue
  (return 0))

(fn ConstructionScript)
