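; Looping gameplay cue that drives the Stamina Hawaii beam's effects and audio on the target pawn.
; Loop start caches the particle components, pawn, Stamina component and beam start time, and starts a repeating timer for team-aware sound indicators.
; Tick feeds the component's endpoint, or replicated intersection point during beam battles, into the particle beam and line audio shape.
; Outside battles the beam fades after a set time; removal fades the audio and clears the timer.
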(event EventOnApplication(EitherVFX-preferthisone)
  (params (MyTarget :type Actor%20Object%20Reference) (Parameters :type Gameplay%20Cue%20Parameters%20Structure%20%28by%20ref%29) (ParticleComponents :type Array%20of%20FXSystem%20Component%20Object%20References) (AudioComponents :type Array%20of%20Audio%20Component%20Object%20References) (BurstCameraShakeInstance :type Camera%20Shake%20Base%20Object%20Reference) (BurstDecalInstance :type Decal%20Actor%20Object%20Reference))
  (bind _fortaudioshapeline (Variables|Default|GetFortAudioShapeLine))
  (Utilities|IsValid _fortaudioshapeline
    (:Is%20Valid
      (Shape|Enable _fortaudioshapeline))))

(event EventTick
  (params (DeltaSeconds :type Float%20%28single-precision%29))
  ;[1 1 1] Tick
  (if (CallFunction|CanUpdateBeam)
    (CallFunction|UpdateBeam)))

(event EventOnLoopingStart(EitherVFX-preferthisversion)
  (params (MyTarget :type Actor%20Object%20Reference) (Parameters :type Gameplay%20Cue%20Parameters%20Structure%20%28by%20ref%29) (ParticleComponents :type Array%20of%20FXSystem%20Component%20Object%20References) (AudioComponents :type Array%20of%20Audio%20Component%20Object%20References))
  ;[1 0.8659 0.4917] Start FX
  (Utilities|FlowControl|Sequence
    (:then_0
      ;;[1 1 1] Cache FX
      (bind _result (ArrayMacros|SafeArrayGet ParticleComponents)
        (:%20%20
          (Utilities|IsValid _result
            (:Is%20Valid
              (Variables|Default|SetLoopingBeam _result))))))
    (:then_1
      (bind _result_2 (ArrayMacros|SafeArrayGet ParticleComponents 1)
        (:%20%20
          (Utilities|IsValid _result_2
            (:Is%20Valid
              (Variables|Default|SetLoopingMuzzleFlash _result_2))))))
    (:then_2
      ;;[1 1 1] Cache pawn
      (Utilities|IsValid MyTarget
        (:Is%20Valid
          (bind _asplayer_pawn_athena (Utilities|Casting|CastToPlayerPawn_Athena MyTarget))
          (Variables|Default|SetTargetPawn _asplayer_pawn_athena)
          (bind _output_get (Variables|Default|SetStaminaComp (Actor|GetComponentbyClass (Variables|Default|GetTargetPawn) "/StaminaGameplay/Gameplay/Items/Hawaii/Stamina_Hawaii_Component.Stamina_Hawaii_Component_C")))
          (Variables|Default|SetStartTime (Variables|/StaminaGameplay/Gameplay/Items/Hawaii/Stamina_Hawaii_Component.Stamina_Hawaii_Component_C|GetBeamStartTime _output_get)))))
    (:then_3
      (bind _returnvalue (Utilities|Time|SetTimerbyEvent (delegate Custom|ShowShoundIndicator) 0.3 true))
      (Variables|Default|SetSoundIndicatorTimer _returnvalue))))

(event Custom|ShowShoundIndicator
  (Fortnite|ShowSoundIndicatorLocalforTeam 0 0 (Transformation|GetActorLocation (Variables|Default|GetTargetPawn)) (Variables|/Script/Engine.SoundBase|GetMaxDistance (Variables|/Script/AudioShapes.AudioShapePrimitiveComponent|GetSoundOnEdge (Variables|Default|GetFortAudioShapeLine))) "Gunshot" (Utilities|Array|MakeArray (Fortnite|GetFortPlayerControllerfromActor (Variables|Default|GetTargetPawn))) (Team|GetActorTeam (Variables|Default|GetTargetPawn)) (Utilities|Array|MakeArray "Friendly" "Neutral" "Hostile")))

(event Custom|UpdateBeam
  (bind _targetpawn (Variables|Default|GetTargetPawn))
  (bind _returnvalue (Actor|GetComponentbyClass _targetpawn "/StaminaGameplay/Gameplay/Items/Hawaii/Stamina_Hawaii_Component.Stamina_Hawaii_Component_C"))
  ;[0.5595 1 0.4855] Beam Tick
  (Utilities|IsValid _returnvalue
    (:Is%20Valid
      (if (Variables|/StaminaGameplay/Gameplay/Items/Hawaii/Stamina_Hawaii_Component.Stamina_Hawaii_Component_C|GetRepInBeamBattle _returnvalue)
        (Variables|Default|SetEndpoint (Variables|/StaminaGameplay/Gameplay/Items/Hawaii/Stamina_Hawaii_Component.Stamina_Hawaii_Component_C|GetRepBattleIntersectionPoint _returnvalue))
        (Variables|Default|SetFadeOutParam 0)
        (else
          (bind _location (CallFunction|TraceforBeamEndpoint))
          (Variables|Default|SetEndpoint _location)
          (if (> (- (Utilities|Time|GetTimeSeconds 0) (Variables|Default|GetStartTime)) 1.8)
            (Variables|Default|SetFadeOutParam 1)
            (else
              (Variables|Default|SetFadeOutParam 0)))))
      (Effects|Components|ParticleSystem|SetVectorParameter (Variables|Default|GetLoopingBeam) "BeamEnd_0" (Variables|Default|GetEndpoint))
      (Effects|Components|ParticleSystem|SetIntParameter (Variables|Default|GetLoopingBeam) "BeamFadeOut" (Variables|Default|GetFadeOutParam))
      (bind _fortaudioshapeline (Variables|Default|GetFortAudioShapeLine)
        (:then
          (Control|SetStartPoint _fortaudioshapeline (Transformation|GetActorLocation))
          (Control|SetEndPoint _fortaudioshapeline (Variables|Default|GetEndpoint)))
        (:else)))))

(event EventOnRemoval(EitherVFX-preferthisversion)
  (params (MyTarget :type Actor%20Object%20Reference) (Parameters :type Gameplay%20Cue%20Parameters%20Structure%20%28by%20ref%29) (ParticleComponents :type Array%20of%20FXSystem%20Component%20Object%20References) (AudioComponents :type Array%20of%20Audio%20Component%20Object%20References) (BurstCameraShakeInstance :type Camera%20Shake%20Base%20Object%20Reference) (BurstDecalInstance :type Decal%20Actor%20Object%20Reference))
  (bind _fortaudioshapeline (Variables|Default|GetFortAudioShapeLine))
  (Utilities|IsValid _fortaudioshapeline
    (:Is%20Valid
      (Variables|/Script/AudioShapes.AudioShapeComponent|SetFadeOutTime 1.0 _fortaudioshapeline)
      (Shape|Disable _fortaudioshapeline)
      (Utilities|Time|ClearandInvalidateTimerbyHandle 0 (Variables|Default|GetSoundIndicatorTimer)))))

(fn CanUpdateBeam
  (return (Math|Boolean|ANDBoolean (Utilities|IsValid (Variables|Default|GetLoopingBeam) :!class /Script/Engine.KismetSystemLibrary) (Utilities|IsValid (Variables|Default|GetTargetPawn) :!class /Script/Engine.KismetSystemLibrary) (Utilities|IsValid (Variables|Default|GetLoopingMuzzleFlash) :!class /Script/Engine.KismetSystemLibrary))))

(fn TraceForBeamEndpoint
  (bind _staminacomp (Variables|Default|GetStaminaComp))
  (Utilities|IsValid _staminacomp
    (:Is%20Valid
      (return (Variables|/StaminaGameplay/Gameplay/Items/Hawaii/Stamina_Hawaii_Component.Stamina_Hawaii_Component_C|GetBeamEndForGC _staminacomp)))))

(fn ConstructionScript)
