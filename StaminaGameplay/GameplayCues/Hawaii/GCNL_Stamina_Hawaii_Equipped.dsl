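; Looping gameplay cue for the equipped Stamina Hawaii item that keeps its equip effect in step with the beam ability.
; On loop start it caches a particle component from the cue and the target Athena player pawn.
; Each tick it polls the pawn's Hawaii component for replicated beam data and cooldown blocking.
; It derives a ready, charging, firing or cooldown state and passes it to the effect's EquipState integer parameter.
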(event EventBeginPlay
  (|CallParentFunction))

(event EventOnLoopingStart(EitherVFX-preferthisversion)
  (params (MyTarget :type Actor%20Object%20Reference) (Parameters :type Gameplay%20Cue%20Parameters%20Structure%20%28by%20ref%29) (ParticleComponents :type Array%20of%20FXSystem%20Component%20Object%20References) (AudioComponents :type Array%20of%20Audio%20Component%20Object%20References))
  (bind _result (ArrayMacros|SafeArrayGet ParticleComponents)
    (:%20%20
      (Utilities|IsValid _result
        (:Is%20Valid
          (Variables|Default|SetEquipFX _result)
          (bind _asplayer_pawn_athena (Utilities|Casting|CastToPlayerPawn_Athena MyTarget))
          (Variables|Components|SetTargetPawn _asplayer_pawn_athena))))))

(event EventTick
  (params (DeltaSeconds :type Float%20%28single-precision%29))
  (bind _targetpawn (Variables|Components|GetTargetPawn))
  (bind _returnvalue (Actor|GetComponentbyClass _targetpawn "/StaminaGameplay/Gameplay/Items/Hawaii/Stamina_Hawaii_Component.Stamina_Hawaii_Component_C"))
  (bind _repbeamdata (Variables|/StaminaGameplay/Gameplay/Items/Hawaii/Stamina_Hawaii_Component.Stamina_Hawaii_Component_C|GetRepBeamData _returnvalue))
  (bind (_beamactive_1_74f47b0849047809ab388489df33bd1b _abilityactive_7_97a41cf944e93caf32bea3a685a6da71) (Utilities|Struct|BreakStaminaHawaiiBeamData _repbeamdata))
  (bind _returnvalue_2 (not _beamactive_1_74f47b0849047809ab388489df33bd1b))
  (bind _blockedbycooldown (Variables|/StaminaGameplay/Gameplay/Items/Hawaii/Stamina_Hawaii_Component.Stamina_Hawaii_Component_C|GetBlockedByCooldown _returnvalue))
  (Utilities|IsValid (Variables|Default|GetEquipFX)
    (:Is%20Valid
      (Utilities|IsValid (Variables|Components|GetTargetPawn)
        (:Is%20Valid
          (Utilities|IsValid _returnvalue
            (:Is%20Valid
              (Utilities|FlowControl|Sequence
                (:then_0
                  (if (and _returnvalue_2 (not _blockedbycooldown))
                    (if (== (Variables|Default|GetEquipStateParam) 3)
                      _)
                    ;[1 1 1] Ready To Fire
                    (Variables|Default|SetEquipStateParam 0)
                    (elif (and _abilityactive_7_97a41cf944e93caf32bea3a685a6da71 _returnvalue_2)
                      ;[1 1 1] Charging
                      (Variables|Default|SetEquipStateParam 1)
                      (elif _beamactive_1_74f47b0849047809ab388489df33bd1b
                        ;[1 1 1] Firing
                        (Variables|Default|SetEquipStateParam 2)
                        (elif _blockedbycooldown
                          ;[1 1 1] Cooldown
                          (Variables|Default|SetEquipStateParam 3))))))
                (:then_1
                  (Effects|Components|ParticleSystem|SetIntParameter (Variables|Default|GetEquipFX) "EquipState" (Variables|Default|GetEquipStateParam)))))
            (:Is%20Not%20Valid
              (Variables|Default|SetEquipStateParam 0))))))))

(fn ConstructionScript)
