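; Looping gameplay cue for the Stamina "Hawaii" item's glint, which supplies validity and endpoint logic for a beam effect.
; CanUpdateBeam passes only when the looping beam and target pawn are both valid.
; TraceForBeamEndpoint runs no trace itself; it returns BeamEndForGC from the Stamina Hawaii component when that component is valid.
; The event handlers, including looping start and tick, are empty here.
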(event Collision|EventActorBeginOverlap
  (params (OtherActor :type Actor%20Object%20Reference)))

(event EventBeginPlay)

(event EventTick
  (params (DeltaSeconds :type Float%20%28single-precision%29)))

(event EventOnLoopingStart(EitherVFX-preferthisversion)
  (params (MyTarget :type Actor%20Object%20Reference) (Parameters :type Gameplay%20Cue%20Parameters%20Structure%20%28by%20ref%29) (ParticleComponents :type Array%20of%20FXSystem%20Component%20Object%20References) (AudioComponents :type Array%20of%20Audio%20Component%20Object%20References)))

(fn CanUpdateBeam
  (return (and (Utilities|IsValid (Variables|Default|GetLoopingBeam) :!class /Script/Engine.KismetSystemLibrary) (Utilities|IsValid (Variables|Default|GetTargetPawn) :!class /Script/Engine.KismetSystemLibrary))))

(fn TraceForBeamEndpoint
  (bind _staminacomp (Variables|Default|GetStaminaComp))
  (Utilities|IsValid _staminacomp
    (:Is%20Valid
      (return (Variables|/StaminaGameplay/Gameplay/Items/Hawaii/Stamina_Hawaii_Component.Stamina_Hawaii_Component_C|GetBeamEndForGC _staminacomp)))))

(fn ConstructionScript)
