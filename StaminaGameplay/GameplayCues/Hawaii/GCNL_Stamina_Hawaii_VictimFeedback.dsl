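; Hawaii looping gameplay cue for victim feedback, built to show a repeated warning lens effect.
; Looping start only caches the target and cue parameters and resets the loop counter.
; LensEffectEvent increments the counter, checks it against NumLoops, and uses StaminaFunctionLibrary to spawn the effect from target and effect causer, with a short timer re-entering through RetriggerLens.
; Nothing shown starts that chain, and OnBurst is empty.
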(event EventBeginPlay
  (|CallParentFunction))

(event EventTick
  (params (DeltaSeconds :type Float%20%28single-precision%29)))

(event EventOnLoopingStart(EitherVFX-preferthisversion)
  (params (MyTarget :type Actor%20Object%20Reference) (Parameters :type Gameplay%20Cue%20Parameters%20Structure%20%28by%20ref%29) (ParticleComponents :type Array%20of%20FXSystem%20Component%20Object%20References) (AudioComponents :type Array%20of%20Audio%20Component%20Object%20References))
  (Variables|Default|SetMy%20Target MyTarget)
  (Variables|Default|SetParameters Parameters)
  (Variables|Default|SetLoopCount 0))

(event Custom|LensEffectEvent
  (bind (_normalizedmagnitude _rawmagnitude _effectcontext _matchedtagname _originaltag _aggregatedsourcetags _aggregatedtargettags _location _normal _instigator _effectcauser) (Ability|GameplayCue|BreakGameplayCueParameters (Variables|Default|GetParameters)))
  (Math|Integer|IncrementInt (Variables|Default|GetLoopCount)
    (:%20%20
      (if (< (Variables|Default|GetLoopCount) (Variables|Default|GetNumLoops))
        (Class|StaminaFunctionLibrary|SpawnVictimWarningLensEffect (Variables|Default|GetMy%20Target) _effectcauser)
        (Utilities|Time|SetTimerbyEvent (delegate Custom|RetriggerLens) 0.25)))))

(event Custom|RetriggerLens
  (CallFunction|LensEffectEvent))

(fn NewFunction)

(fn OnBurst
  (params (MyTarget :type Actor%20Object%20Reference) (Parameters :type Gameplay%20Cue%20Parameters%20Structure%20%28by%20ref%29) (ParticleComponents :type Array%20of%20Cascade%20Particle%20System%20Component%20Object%20References) (AudioComponents :type Array%20of%20Audio%20Component%20Object%20References) (BurstCameraShakeInstance :type Camera%20Shake%20Base%20Object%20Reference) (BurstDecalInstance :type Decal%20Actor%20Object%20Reference)))

(fn ConstructionScript)
